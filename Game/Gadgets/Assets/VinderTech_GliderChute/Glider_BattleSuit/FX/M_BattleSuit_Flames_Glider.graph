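# Translucent unlit, two-sided material for animated stylized flames on the BattleSuit glider, usable on skeletal and Niagara meshes.
# Panned fire-mask samples, offset by normal-map distortion, vertex color and a sine mask, form opacity and drive a blue curve-atlas HDR color, with depth fades and optional Fresnel and UV gradients.
# Vertex offset adds two scrolling pulses and a forward stretch, driven by particle data or parameters via a static switch.
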Material /Game/Gadgets/Assets/VinderTech_GliderChute/Glider_BattleSuit/FX/M_BattleSuit_Flames_Glider
graph {
 properties {
  bUsedWithMeshDeformer=true
  bUsedWithNiagaraMeshParticles=true
  bUsedWithSkeletalMesh=true
  blendMode=BLEND_Translucent
  pixelDepthOffsetMode=PDOM_Legacy
  shadingModel=MSM_Unlit
  twoSided=true
  usedShadingModels=Unlit
 }
 nodes [
  n1 = ParticleColor()
  n2 = ScalarParameter() {defaultValue=1, group=Alpha, parameterName=MasterAlpha}
  n3 = VertexColor()
  n4 = ScalarParameter() {defaultValue=1.15, parameterName=Param_1}
  n5 = Add(A=TextureCoordinate(), B=AppendVector(A=Multiply(A=@n3.R, B=@n4), B=Constant()))
  n6 = TextureSample(UVs=Panner(Coordinate=TextureCoordinate()) {speedX=-1, speedY=0.1}) {
   samplerSource=SSM_Wrap_WorldGroupSettings
   samplerType=SAMPLERTYPE_Normal
   texture=/Game/Effects/Fort_Effects/Gameplay/SmokeDustExplosions/Textures/T_Explosion_NM_01
  }
  n7 = ComponentMask(None=MaterialFunctionCall(Normal=@n6.RGB, Flatness=ScalarParameter() {defaultValue=1}) {materialFunction=&FlattenNormal}.Result) {g=true, r=true}
  n8 = Add(A=@n5, B=@n7)
  n9 = ComponentMask(None=@n8) {r=true}
  n10 = Add(A=ComponentMask(None=@n8) {g=true}) {constB=-0.15}
  n11 = ComponentMask(None=@n8) {r=true}
  n12 = Add(A=ComponentMask(None=@n8) {g=true}) {constB=0.45}
  n13 = Add(A=TextureCoordinate() {uTiling=2}, B=Multiply(A=VertexColor().R, B=ScalarParameter() {defaultValue=1.15, parameterName=Param_1}))
  n14 = VectorParameter() {group=AnimeFlameShape, parameterName=Anime Flame SineMask Panner Speed, sortPriority=3, defaultValue=<rgba -5 0 0 0>}
  n15 = ComponentMask(None=@n14.RGB) {g=true, r=true}
  n16 = LinearInterpolate(Alpha=TextureSample(UVs=Panner(Coordinate=@n13, Speed=@n15)) {samplerType=SAMPLERTYPE_Masks, texture=/Engine/Functions/Engine_MaterialFunctions02/ExampleContent/Textures/SineWaveHeightMap}.R) {constA=0.5, constB=1.5}
  n17 = VectorParameter() {group=AnimeFlameShape, parameterName=Anime Flame Shape Panner Speed, sortPriority=3, defaultValue=<rgba -1 0 0 0>}
  n18 = ComponentMask(None=@n17.RGB) {g=true, r=true}
  n19 = Panner(Coordinate=LinearInterpolate(A=AppendVector(A=@n9, B=@n10), B=AppendVector(A=@n11, B=@n12), Alpha=@n16), Speed=@n18)
  n20 = ScalarParameter() {parameterName=Bias_A}
  n21 = Add(A=@n5, B=@n7)
  n22 = ComponentMask(None=@n21) {r=true}
  n23 = Add(A=ComponentMask(None=@n21) {g=true}) {constB=-0.15}
  n24 = ComponentMask(None=@n21) {r=true}
  n25 = Add(A=ComponentMask(None=@n21) {g=true}) {constB=0.45}
  n26 = Panner(Coordinate=LinearInterpolate(A=AppendVector(A=@n22, B=@n23), B=AppendVector(A=@n24, B=@n25), Alpha=@n16), Speed=@n18)
  n27 = Multiply(A=Multiply(
    A=TextureSample!(UVs=@n19, Tex, Bias=@n20, Apply View MipBias) {mipValueMode=TMVM_MipBias, texture=/Game/Effects/Fort_Effects/Textures/Fire/T_Fire_Mask}.R
    B=TextureSample(UVs=@n26, Bias=@n20) {mipValueMode=TMVM_MipBias, texture=&T_Fire_Mask}.R
   )) {constB=0.333}
  n28 = Add(A=@n5, B=@n7)
  n29 = ComponentMask(None=@n28) {r=true}
  n30 = Add(A=ComponentMask(None=@n28) {g=true}) {constB=-0.15}
  n31 = ComponentMask(None=@n28) {r=true}
  n32 = Add(A=ComponentMask(None=@n28) {g=true}) {constB=0.25}
  n33 = VectorParameter() {group=AnimeFlameShape, parameterName=Anime Flame Shape Panner Speed, sortPriority=3, defaultValue=<rgba -1 0 0 0>}
  n34 = ComponentMask(None=@n33.RGB) {g=true, r=true}
  n35 = Panner(Coordinate=LinearInterpolate(A=AppendVector(A=@n29, B=@n30), B=AppendVector(A=@n31, B=@n32), Alpha=@n16), Speed=@n34)
  n36 = ScalarParameter() {parameterName=Bias_A}
  n37 = Add(A=@n5, B=@n7)
  n38 = ComponentMask(None=@n37) {r=true}
  n39 = Add(A=ComponentMask(None=@n37) {g=true}) {constB=-0.15}
  n40 = ComponentMask(None=@n37) {r=true}
  n41 = Add(A=ComponentMask(None=@n37) {g=true}) {constB=0.25}
  n42 = Panner(Coordinate=LinearInterpolate(A=AppendVector(A=@n38, B=@n39), B=AppendVector(A=@n40, B=@n41), Alpha=@n16), Speed=@n34)
  n43 = Multiply(A=Multiply(
    A=TextureSample(UVs=@n35, Bias=@n36) {mipValueMode=TMVM_MipBias, texture=&T_Fire_Mask}.R
    B=TextureSample(UVs=@n42, Bias=@n36) {mipValueMode=TMVM_MipBias, texture=&T_Fire_Mask}.R
   )) {constB=0.333}
  n44 = Add(A=@n5, B=@n7)
  n45 = ComponentMask(None=@n44) {g=true}
  n46 = LinearInterpolate(
   A=AppendVector(A=ComponentMask(None=@n44) {r=true}, B=Add(A=@n45) {constB=-0.15})
   B=AppendVector(A=ComponentMask(None=@n44) {r=true}, B=ComponentMask(None=@n44) {g=true})
   Alpha=@n16
  )
  n47 = VectorParameter() {group=AnimeFlameShape, parameterName=Anime Flame Shape Panner Speed, sortPriority=3, defaultValue=<rgba -1 0 0 0>}
  n48 = ComponentMask(None=@n47.RGB) {g=true, r=true}
  n49 = ScalarParameter() {parameterName=Bias_A}
  n50 = Add(A=@n5, B=@n7)
  n51 = ComponentMask(None=@n50) {g=true}
  n52 = LinearInterpolate(
   A=AppendVector(A=ComponentMask(None=@n50) {r=true}, B=Add(A=@n51) {constB=-0.15})
   B=AppendVector(A=ComponentMask(None=@n50) {r=true}, B=ComponentMask(None=@n50) {g=true})
   Alpha=@n16
  )
  n53 = Multiply(
   A=TextureSample(UVs=Panner(Coordinate=@n46, Speed=@n48), Bias=@n49) {mipValueMode=TMVM_MipBias, texture=&T_Fire_Mask}.R
   B=TextureSample(UVs=Panner(Coordinate=@n52, Speed=@n48), Bias=@n49) {mipValueMode=TMVM_MipBias, texture=&T_Fire_Mask}.R
  )
  n54 = Clamp(None=Add(A=Add(A=@n27, B=@n43), B=Multiply(A=@n53) {constB=0.333}))
  n55 = TextureCoordinate()
  n56 = StaticSwitchParameter(True=Sine(None=ComponentMask(None=@n55) {r=true}) {period=2}, False=Constant() {r=1}) {parameterName=Use Particle Opacity}
  n57 = Multiply(
   A=Multiply(
    A=StaticSwitchParameter(True=@n1.A, False=@n2) {parameterName=Use Particle Opacity}
    B=MaterialFunctionCall() {materialFunction=&CameraDepthFade}.Result
   )
   B=DepthFade(Opacity=Multiply(A=@n54, B=@n56), FadeDistance=ScalarParameter() {defaultValue=2, group=Alpha, parameterName=Fade Distance, sortPriority=1})
  )
  n58 = OneMinus(None=Fresnel(ExponentIn=ScalarParameter() {defaultValue=5, group=Opacity_Fresnel, parameterName=Fresnel_ExponentIn}) {baseReflectFraction=0, exponent=0.1})
  n59 = ScalarParameter() {defaultValue=2, group=Opacity_Fresnel, parameterName=Fresnel_Power}
  n60 = LinearInterpolate(A=@n57, B=Multiply(A=@n57, B=Power(Base=@n58, Exp=@n59)), Alpha=ScalarParameter() {group=Opacity_Fresnel, parameterName=UseFresnel})
  n61 = Power(
   Base=ComponentMask(None=TextureCoordinate()) {g=true}
   Exp=ScalarParameter() {defaultValue=1, group=Opacity_Fresnel, parameterName=UGradient_PWR}
  )
  n62 = OneMinus(None=ComponentMask(None=TextureCoordinate()) {g=true})
  n63 = ScalarParameter() {group=Opacity_Fresnel, parameterName=VGradient_Offset}
  n64 = Power(Base=Saturate(None=Add(A=@n62, B=@n63)), Exp=ScalarParameter() {defaultValue=1, group=Opacity_Fresnel, parameterName=VGradient_PWR})
  LinearInterpolate(A=@n60, B=Multiply(A=@n60, B=Min(A=@n61, B=@n64)), Alpha=ScalarParameter() {group=Opacity_Fresnel, parameterName=UVGradient_Opacity})
  n65 = CurveAtlasRowParameter(CurveTime=Power(Base=@n54, Exp=ScalarParameter() {defaultValue=1, group=Color, parameterName=Curve Contrast Controller, sortPriority=1})) {
   atlas=&AthenaCurveAtlas01
   curve=/Game/Characters/Player/Male/Medium/Bodies/M_MED_BattleSuit/FX/BattleSuit_Blue
   defaultValue=82
   group=Color
   parameterName=Curve Parameter
  }
  n66 = Desaturation(None=@n65)
  n67 = ScalarParameter() {parameterName=Desaturation}
  n68 = Multiply(A=@n65, B=VectorParameter() {group=Color, parameterName=HDR_Tint, sortPriority=2, defaultValue=<rgba 1 1 2 0>}.RGB)
  n69 = MaterialFunctionCall(
   LDR Input=Clamp(None=LinearInterpolate(A=@n65, B=@n66, Alpha=@n67))
   HDR Tint=LinearInterpolate(A=@n68, B=Desaturation(None=@n68), Alpha=@n67)
   HDR Intensity Multiplier=ScalarParameter() {defaultValue=1, group=Color, parameterName=HDR_Intensity, sortPriority=3}
  ) {materialFunction=&DeriveHDRfromLDR}
  MaterialFunctionCall(Emissive=StaticSwitchParameter(True=Multiply(A=@n69.RemappedRange, B=@n1.RGB), False=@n69.RemappedRange) {parameterName=Use Particle Opacity}) {materialFunction=&MF_FX_EmissiveAdjustments}
  n70 = VectorParameter() {group=WPO_AnimeFlame, parameterName=Regular Pulse Vector, sortPriority=2, defaultValue=<rgba 0 1 0 1>}
  n71 = Transform(None=@n70.RGB) {transformSourceType=TRANSFORMSOURCE_Local}
  n72 = TextureCoordinate()
  n73 = VertexColor()
  n74 = ScalarParameter() {defaultValue=3.6667, parameterName=WPO_A}
  n75 = Add(A=TextureCoordinate() {uTiling=2}, B=AppendVector(A=Multiply(A=@n73.R, B=@n74), B=Constant()))
  n76 = VectorParameter() {group=WPO_AnimeFlame, parameterName=Regular Pulse Panner Speed, sortPriority=4, defaultValue=<rgba 0 -2 0 0>}
  n77 = TextureSample!(
   UVs=Panner(
    Coordinate=MaterialFunctionCall(XY=@n75) {materialFunction=&Swizzle}.YX
    Speed=ComponentMask(None=@n76.RGB) {g=true, r=true}
   ) {speedY=10}
   Tex
   Apply View MipBias
  ) {samplerSource=SSM_Wrap_WorldGroupSettings, texture=/Game/Global/Textures/Default/Masks_and_Gradients/gradient_stasis}
  n78 = Multiply(A=OneMinus(None=ComponentMask(None=@n72) {g=true}), B=Multiply(A=@n77.R, B=@n77.R))
  n79 = Multiply(
   A=Multiply(A=Multiply(A=@n71, B=@n78), B=ScalarParameter() {defaultValue=100, group=WPO_AnimeFlame, parameterName=Regular Pulse WPO Amount})
   B=ComponentMask(None=TextureCoordinate()) {r=true}
  )
  n80 = VectorParameter() {group=WPO_AnimeFlame, parameterName=Staccato Pulse Vector, sortPriority=3, defaultValue=<rgba 0 1 0 1>}
  n81 = OneMinus(None=ComponentMask(None=TextureCoordinate()) {g=true})
  n82 = VertexColor()
  n83 = ScalarParameter() {defaultValue=7, parameterName=WPO_B}
  n84 = Add(A=TextureCoordinate(), B=AppendVector(A=Multiply(A=@n82.R, B=@n83), B=Constant()))
  n85 = VectorParameter() {group=WPO_AnimeFlame, parameterName=Staccato Pulse Panner Speed, sortPriority=5, defaultValue=<rgba 0 -1 0 0>}
  n86 = TextureSample(UVs=Panner(
    Coordinate=MaterialFunctionCall(XY=@n84) {materialFunction=&Swizzle}.YX
    Speed=ComponentMask(None=@n85.RGB) {g=true, r=true}
   )) {samplerSource=SSM_Wrap_WorldGroupSettings, texture=/Game/Global/Textures/Default/Masks_and_Gradients/T_Vert_Gradient_Sections}
  n87 = Multiply(A=@n86.R, B=@n86.R)
  n88 = Multiply(
   A=Multiply(A=Transform(None=@n80.RGB) {transformSourceType=TRANSFORMSOURCE_Local}, B=Multiply(A=@n81, B=@n87))
   B=ScalarParameter() {defaultValue=20, group=WPO_AnimeFlame, parameterName=Staccato Pulse WPO Amount, sortPriority=1}
  )
  n89 = TextureCoordinate()
  n90 = Multiply(
   A=OneMinus(None=ComponentMask(None=@n89) {g=true})
   B=Transform(None=VectorParameter() {parameterName=Glider Vector, defaultValue=<rgba 1 1 0 1>}.RGB) {transformSourceType=TRANSFORMSOURCE_Local}
  )
  n91 = LinearInterpolate(
   B=ScalarParameter() {defaultValue=25, parameterName=FullStretch}
   Alpha=StaticSwitchParameter(True=DynamicParameter() {paramNames [Stretch,Param2,Param3,Param4]}.Stretch, False=ScalarParameter() {parameterName=Glider Forward}) {parameterName=Use Particle Opacity}
  )
  n92 = TextureCoordinate()
  Multiply(A=Add(A=Add(A=@n79, B=@n88), B=Multiply(A=@n90, B=@n91)), B=Sine(None=ComponentMask(None=@n92) {r=true}) {period=2})
  ScalarParameter() {defaultValue=1}
 ]
}
